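# Translucent unlit material for a pickaxe swing trail, flagged for Niagara ribbons and located in a Test folder.
# Color blends time-of-day tints via Color_Time_of_Day, multiplied by particle color and passed through MF_FX_EmissiveAdjustments.
# Opacity multiplies a panning smoke-trail texture, panning caustics noise, particle alpha and a sine fade along U, which a parameter can disable.
# A shading-path switch applies depth fade on default and deferred paths, skipping it on forward and mobile.
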Material /Game/Weapons/FORT_Melee/Pickaxe_InstantGravelNoble/Meshes/Test/M_Pickaxe_Ultralight_female_swing_trail
graph {
 properties {bUsedWithNiagaraRibbons=true, blendMode=BLEND_Translucent, pixelDepthOffsetMode=PDOM_Legacy, shadingModel=MSM_Unlit, usedShadingModels=Unlit}
 nodes [
  n1 = VectorParameter() {parameterName="Color_Morning (V3)", defaultValue=<rgba 1 1 1 0>}
  n2 = Constant() {r=1}
  n3 = VectorParameter() {parameterName="Color_Evening (V3)", defaultValue=<rgba 1 1 1 0>}
  n4 = VectorParameter() {parameterName="Color_Night (V3)", defaultValue=<rgba 1 0.9961 0.757 0>}
  n5 = ParticleColor()
  MaterialFunctionCall(Emissive=Multiply(
    A=MaterialFunctionCall(Color_Morning=@n1.RGB, Color_Day=@n2, Color_Evening=@n3.RGB, Color_Night=@n4.RGB) {materialFunction=&Color_Time_of_Day}.Result
    B=@n5.RGB
   )) {materialFunction=&MF_FX_EmissiveAdjustments}
  n6 = VectorParameter() {parameterName=ReUseParam_UVScale, defaultValue=<rgba 1.3 1 0 1>}
  n7 = VectorParameter() {parameterName=ReUseParam_UVPanner, defaultValue=<rgba -0.6 0 0 1>}
  n8 = Panner(Coordinate=Multiply(A=TextureCoordinate(), B=AppendVector(A=@n6.R, B=@n6.R)), Speed=AppendVector(A=@n7.R, B=@n7.R)) {speedX=-0.2}
  n9 = Panner(Coordinate=Add(A=TextureCoordinate() {vTiling=0.5}) {constB=0.3}) {speedX=0.2}
  n10 = Multiply(
   A=Power(Base=TextureSample(UVs=@n8) {samplerType=SAMPLERTYPE_Grayscale, texture=&T_Smoke_Trail_Soft}.R)
   B=Multiply(A=TextureSample(UVs=@n9) {samplerType=SAMPLERTYPE_LinearGrayscale, texture=/Game/Global/Textures/Default/Caustics/T_Nosie_Caustics_Grayscale}.R) {constB=2}
  ) {constA=1}
  n11 = Sine(None=ComponentMask(None=TextureCoordinate()) {r=true}) {period=2}
  n12 = Multiply(
   A=Multiply(A=@n5.A, B=Add(B=@n10))
   B=LinearInterpolate(A=Power(Base=@n11) {constExponent=1}, Alpha=ScalarParameter() {parameterName=ReUseParam_DisableTipFade}) {constAlpha=0}
  )
  n13 = DepthFade(Opacity=@n12) {fadeDistanceDefault=20}
  ShadingPathSwitch(Default=@n13, Deferred=@n13, Forward=@n12, Mobile=@n12)
 ]
}
